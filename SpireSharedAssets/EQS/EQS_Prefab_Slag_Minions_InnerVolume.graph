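# Environment query for an inner volume of Slag minions, going by its name, that generates candidate points around the querier.
# A donut generator with equal inner and outer radius yields a single evenly spaced ring, with spiral pattern off.
# Projection uses an NPC navigation filter that excludes all water.
# The query has no tests, so the generated points are neither scored nor further filtered.
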EnvQuery /SpireSharedAssets/EQS/EQS_Prefab_Slag_Minions_InnerVolume
options [{
  generator {
   name=EnvQueryGenerator_Donut_0
   class=EnvQueryGenerator_Donut
   properties {
    bUseSpiralPattern=false
    center=&EnvQueryContext_Querier
    innerRadius=1200
    numberOfRings=1
    outerRadius=1200
    pointsPerRing=16
    projectionData {
     extentX=0
     navigationFilter=&NavFilter_NPC_ExcludeAllWater_C
     postProjectionVerticalOffset=0
    }
   }
  }
  tests []
 }]
