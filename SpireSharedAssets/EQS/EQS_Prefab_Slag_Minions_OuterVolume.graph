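# EQS query that finds flat spots in a wide outer band around the querier, named for placing Slag minion prefabs.
# A donut generator builds concentric rings of candidate points at long range, projected by geometry trace rather than onto navigation.
# A Fortnite flat-surface test then evaluates each point by tracing around it within a height tolerance.
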EnvQuery /SpireSharedAssets/EQS/EQS_Prefab_Slag_Minions_OuterVolume
options [{
  generator {
   name=EnvQueryGenerator_Donut_0
   class=EnvQueryGenerator_Donut
   properties {
    bUseSpiralPattern=false
    center=&EnvQueryContext_Querier
    innerRadius=7500
    numberOfRings=3
    outerRadius=12500
    pointsPerRing=24
    projectionData {extentX=0, navigationFilter=null, postProjectionVerticalOffset=0, traceMode=GeometryByChannel}
   }
  }
  tests [FortQueryTest_OnFlatSurface_1 {class=FortQueryTest_OnFlatSurface, properties {numberOfIteration=2, radius=200, toleranceZ=150, traceOffsetDown=200, traceOffsetUp=200}}]
 }]
